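# Niagara activation effect for the ClawPad Host glider, adding air trails and a display light.
# Main_Trail draws short-lived anim-trail ribbons that fade with system age, while Tip_Trails spawns socket-partitioned ribbons whose drift velocity eases to zero.
# Display_Light bursts socket-attached sprites in mirrored spawn groups, computing glow brightness from camera angle via a cross product.
# Front-end modifiers swap velocities, offsets and opacity from gameplay values, with time-of-day scaling throughout.
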NiagaraSystem /BRCosmetics/Gadgets/Assets/VinderTech_GliderChute/Glider_ClawPad_Host/FX/NS_Glider_ClawPad_Host_Activation
dependencies {
 usedDataInterfaces [
  &NiagaraDataInterfaceColorCurve
  &NiagaraDataInterfaceCurve
  &NiagaraDataInterfaceCamera
  &NiagaraDataInterfaceSocketReader
 ]
 usedDynamicInputs [&ColorFromCurve]
 usedModules [
  &self:SystemSpawnScript.SystemScriptSource.SystemScriptGraph.NiagaraNodeAssignment_3.SetVariables_#1
  &self:SystemSpawnScript.SystemScriptSource.SystemScriptGraph.NiagaraNodeAssignment_4.SetVariables_#2
  &SystemState
 ]
}
emitters {
 Display_Light {
  emitterUpdateScript [
   &EmitterState(
    Inactive Response=::NewEnumerator1
    Life Cycle Mode=::NewEnumerator1
    Loop Behavior=::NewEnumerator1
    Loop Duration Mode=::NewEnumerator1
    UseLoopDelay=false
   )
   &SpawnBurst_Instantaneous(Use Spawn Probability=false, UseLoopCountLimit=false)
  ]
  particleSpawnScript [
   &FN_SetSpawnGroup(Number of Spawn Groups=2)
   /SetVariables(Particles.MirrorAxis:Vector3f=&MaskVectorBySpawnGroup(Masked Vector=<-1 1 1>, Passthrough Vector=<1 1 1>)) {isSetParameters=true}
   &InitializeParticle(
    Color=<rgba 0.3906 0.4569 0.6 0.3>
    Mesh Scale Mode=::NewEnumerator3
    Sprite Size=<250 250>
    Sprite Size Mode=::NewEnumerator4
    UsePositionOffset=true
    Write Material Random=true
   )
   &CameraOffset(Camera Offset Amount=11)
   &DynamicMaterialParameters(
    Index 0 Param 1=&Multiply_Float(A=-19, B=&MakeFloatFromVector(Vector=$Particles.MirrorAxis, Channel=::NewEnumerator0))
    Index 0 Param 2=-1
    Module.Param0WriteEnabled=true
    Module.Param1WriteEnabled=true
    Module.Param2WriteEnabled=false
    Module.Param3WriteEnabled=false
   )
  ]
  particleUpdateScript [
   &ParticleState(Kill Particles When Lifetime Has Elapsed=false)
   &SocketLocation(
    Orientation Sampling=::NewEnumerator1
    SamplePosOffset=false
    Socket Index=2
    Socket Reader=$System.Socket Reader
    Socket Sampling Mode=::NewEnumerator3
    VelocitySpeedLimit=false
   )
   &ScaleColor(
    Scale Alpha=&Multiply_Float(
     A=1
     B=&FN_FrontEndFloatModifier(Gameplay=&FNTimeOfDayScalar(Night Scale=1, Morning Scale=0.8, Evening Scale=0.7), Front End=1)
    )
    ScaleA=true
    ScaleRGB=true
   )
   /SetVariables(
    Particles.GlowCameraOffset:NiagaraFloat=&FN_FrontEndFloatModifier(Gameplay=0, Front End=0)
    Particles.GlowColor=&MultiplyLinearColor(
     Linear Color=$Particles.Color
     Scale Factor=&MakeVector4_FromFloat(Float=&Multiply_Float(
       A=&Multiply_Float(A=8, B=1)
       B=&ClampFloat(Float=&OneMinusFloat(Float=&Multiply_Float(
          A=0.7
          B=&VectorLength(
           Vector=/Niagara/DynamicInputs/Vectors/CrossProduct(
            Vector A:Vector3f=&ReturnCameraVector(Camera Vectors=::NewEnumerator1, Coordinate Space=::Simulation)
            Vector B:Vector3f=<0 -1 0>
            Normalize Incoming Vectors:NiagaraBool=true
           )
           Fallback Vector=<0 -1 0>
          )
         )))
      ))
    )
    Particles.GlowPosition:NiagaraPosition=&AddVectorToPosition(
     Position=$Particles.Position
     Vector=&Multiply_Vector(A=&FN_FrontEndVector3Modifier(Gameplay=<0 0 0>, Front End=<0 0 0>), B=$Particles.MirrorAxis)
    )
    Particles.GlowSize=<10 10>
    Particles.glowalignement:Vector3f=<1 0 0>
    Particles.glowfacing:Vector3f=<0 1 0>
   ) {isSetParameters=true}
   /SetVariables(
    Particles.BigGlowColor:LinearColor=&MultiplyLinearColor(
     Linear Color=$Particles.Color
     Scale Factor=&MakeVector4_FromFloat(Float=&Multiply_Float(
       A=&Multiply_Float(A=1.5, B=1)
       B=&ClampFloat(Float=&OneMinusFloat(Float=&Multiply_Float(
          A=0.4
          B=&VectorLength(
           Vector=&CrossProduct(
            Vector A:Vector3f=&ReturnCameraVector(Camera Vectors=::NewEnumerator1, Coordinate Space=::Simulation)
            Vector B:Vector3f=<0 -1 0>
            Normalize Incoming Vectors:NiagaraBool=true
           )
           Fallback Vector=<0 -1 0>
          )
         )))
      ))
    )
    Particles.BigGlowSize:Vector2f=&Vector2DFromFloat(Value=40)
   ) {isSetParameters=true}
  ]
  rendererClasses [&NiagaraSpriteRendererProperties]
  renderers [{rendererClass=&NiagaraSpriteRendererProperties}]
 }
 Main_Trail {
  emitterSpawnScript [&FN_RibbonIDOffsetUtility()]
  emitterUpdateScript [
   &EmitterState()
   &FN_SpawnAnimTrail()
  ]
  particleSpawnScript [
   &FN_InitializeAnimTrailBoneIndecies() {module=FN_InitializeAnimTrailBoneIndecies001}
   &FN_InitializeAnimTrail(
    Color=&MakeLinearColorFromVectorAndFloat(
     "Vector (RGB)"=&MakeVectorFromLinearColorRGB(Color=&MultiplyLinearColor(
       Linear Color=<rgba 0.4993 0.5417 0.5337 0.5>
       Scale Factor=&MakeVector4_FromLinearColor(LinearColor=$System.TOD)
      ))
     "Float (Alpha)"=&FN_FrontEndFloatModifier(Gameplay=0.5, Front End=0.2)
    )
    Lifetime=0.15
    Position Offset=&FN_FrontEndVector3Modifier(Gameplay=<0 0 0>, Front End=<0 0 -15>)
   )
   &AddVelocity(Velocity=&FN_FrontEndVector3Modifier(Gameplay=<0 2000 0>, Front End=<0 -1500 100>))
  ]
  particleUpdateScript [
   &ScaleColor(
    Color Value To Scale=$Particles.Color
    Scale Alpha=&FloatFromCurve(FloatCurve=DataInterface() {propertyValues {Curve {keys [<at 1 1>,<at 2.5 0>]}}}, CurveIndex=$Engine.System.Age)
    ScaleA=true
    ScaleRGB=true
   ) {module=ScaleColor001}
   &SolveForcesAndVelocity(Clamp Velocity=false, Limit Acceleration=false)
   &FN_UpdateAnimTrail()
  ]
  rendererClasses [&NiagaraRibbonRendererProperties]
  renderers [{rendererClass=&NiagaraRibbonRendererProperties}]
 }
 Tip_Trails {
  emitterSpawnScript [&FN_RibbonIDOffsetUtility()]
  emitterUpdateScript [
   &EmitterState()
   &SpawnRate(SpawnRate=240, Use Spawn Probability=false)
  ]
  particleSpawnScript [
   &PartitionParticles(Write Partition to Ribbon ID=::NewEnumerator0)
   &InitializeParticle(
    Color=&MakeLinearColorFromVectorAndFloat(
     "Vector (RGB)"=&MakeVectorFromLinearColorRGB(Color=&MultiplyLinearColor(
       Linear Color=<rgba 0.5885 0.5885 0.5885 0.3>
       Scale Factor=&MakeVector4_FromLinearColor(LinearColor=$System.TOD)
      ))
     "Float (Alpha)"=&FN_FrontEndFloatModifier(Gameplay=0.3, Front End=0.2)
    )
    Lifetime=0.3
    Material Random:NiagaraFloat=&Add_Float(A=0, B=0)
    Ribbon Width=15
    Ribbon Width Mode=::NewEnumerator1
    UsePositionOffset=false
    Write Material Random=true
   )
   &SocketLocation(
    "Inherit Velocity (Scale)"=2
    SamplePosOffset=true
    Sampled Position Offset=&FN_FrontEndVector3Modifier(Gameplay=<0 0 0>, Front End=<0 0 0>)
    Socket Index=$Particles.PartitionParticles.Partition
    Socket Reader=$System.Socket Reader
    Socket Sampling Mode=::NewEnumerator3
    VelocitySpeedLimit=false
   ) {module=SocketLocation001}
  ]
  particleUpdateScript [
   &ParticleState()
   /SetVariables(Particles.Velocity=&Lerp_Vector(
     A=&FN_FrontEndVector3Modifier(Gameplay=<0 3000 500>, Front End=<0 -3000 200>)
     B=<0 0 0>
     Alpha=&FloatFromCurve(FloatCurve=DataInterface() {propertyValues {Curve {keys [<at 0.8 0>,<at 1.5 1>]}}}, CurveIndex=$System.Age)
    )) {isSetParameters=true}
   &ScaleColor(
    Scale Alpha=&Multiply_Float(
     A=&FN_FrontEndFloatModifier(Gameplay=0.7, Front End=0.5)
     B=&Multiply_Float(
      A=&FNTimeOfDayScalar(Night Scale=0.6, Morning Scale=0.8, Evening Scale=0.8)
      B=&FloatFromCurve(FloatCurve=DataInterface() {propertyValues {Curve {keys [<at 0 1>,<at 0.4 0>]}}})
     )
    )
    ScaleA=true
    ScaleRGB=true
   ) {module=ScaleColor001}
   &DynamicMaterialParameters(
    Index 0 Param 1=&FloatFromCurve(FloatCurve=DataInterface() {propertyValues {Curve {keys [
         {interpMode=RCIM_Cubic, tangentMode=RCTM_User, time=0.2, value=1}
         {arriveTangent=-2.3194, interpMode=RCIM_Cubic, leaveTangent=-2.3194, tangentMode=RCTM_User, time=0.5, value=0}
        ]}}})
    Index 0 Param 3=-3
    Module.Index0Float0Write=true
    Module.Index0Float1Write=true
    Module.Index0Float2Write=true
    Module.Index0Float3Write=true
   )
   &SolveForcesAndVelocity(Clamp Velocity=false, Limit Acceleration=false)
   &KillParticles(Kill Particles=&SetBoolByFloatComparison(A=$Particles.NormalizedAge, B=1)) {enabled=false}
  ]
  rendererClasses [&NiagaraRibbonRendererProperties]
  renderers [{rendererClass=&NiagaraRibbonRendererProperties}]
 }
}
summary {
 systemName=NS_Glider_ClawPad_Host_Activation
 emitters [
  {emitterName=Main_Trail, rendererClasses [&NiagaraRibbonRendererProperties]}
  {emitterName=Tip_Trails, rendererClasses [&NiagaraRibbonRendererProperties]}
  {emitterName=Display_Light, rendererClasses [&NiagaraSpriteRendererProperties]}
 ]
 userVariables [
  User.bIsFrontEnd {type=NiagaraBool}
  User.bIsFrontEndPreview {type=NiagaraBool}
  User.RibbonIDOffset {type=NiagaraInt32}
 ]
}
